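# Animation Blueprint for the Pickaxe_PJ_Party melee pickaxe skeleton that adds physics-driven secondary motion to its jaw.
# An Anim Dynamics node simulates dyn_jaw on the input pose with an angular spring, angular limits, and scaled component velocity and acceleration.
# The EventGraph holds math deriving ExtVel from negated, scaled player-character velocity, but it is not chained to the update event and does not feed the node's zero external force.
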AnimBlueprint /BRCosmetics/Weapons/FORT_Melee/Pickaxe_PJ_Party/Meshes/SK_Pickaxe_PJ_Party_Export_Skeleton_AnimBP
target_skeleton=&/SK_Pickaxe_PJ_Party_Export_Skeleton
graphs [
 AnimGraph {nodes [
   n1 = Anim Dynamics - Bone: dyn_jaw(ComponentPose=Local To Component(LocalPose=Input Pose()), Alpha=1, ExternalForce="0,0,0") {
    angularDampingOverride=90.9065
    angularSpringConstant=150
    bAngularSpring=true
    alphaBoolBlend {blendOption=Linear}
    boundBone {boneName=dyn_jaw}
    componentLinearAccScale=<5 5 5>
    componentLinearVelScale=<5 5 5>
    physicsBodyDefinitions [{
      boundBone {boneName=dyn_jaw}
      boxExtents=<30 30 30>
      constraintSetup {twistAxis=AxisZ, angularLimitsMax=<90 0 0>, angularLimitsMin=<0 0 0>, angularTarget=<0 0 0>}
      localJointOffset=<0 5 -10>
     }]
   }
   Output Pose(Result=Component To Local(ComponentPose=@n1)) {layerGroup=None}
  ]}
 EventGraph {nodes [
   Event BlueprintUpdateAnimation()
   n1 = GetVelocity!(self:Actor Object Reference=GetPlayerCharacter!(PlayerIndex:int=0) -> (ReturnValue:Character Object Reference)) -> (ReturnValue:Vector)
   n2 = float * float(A=Get ForceScale() -> (ForceScale:double), B=-1)
   Set ExtVel(ExtVel:Vector=vector / float!(A:Vector=vector * float(A=@n1, B=@n2), B:double=1) -> (ReturnValue:Vector)) -> (Output_Get:Vector)
  ]}
]
variables [
 ExtVel {type=Vector}
 ForceScale {type=number, default=10}
]
